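# Niagara swing trail effect for the CarbideBlue pickaxe, drawing blue ribbons and a glow as the weapon moves.
# Ribbon_SwingSmear and Ribbon_Trail use anim-trail modules keyed to bone or socket indices, rendering short-lived, fading ribbons through multiple renderers.
# Particle_Glow bursts socket-placed sprites that grow and fade while driving a light renderer from particle color.
# All three emitters scale alpha by time of day, dimming most at night.
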NiagaraSystem /BRCosmetics/Weapons/FORT_Melee/Pickaxe_CarbideBlue/FX/NS_Pickaxe_CarbideBlue_Trail
dependencies {
 usedDataInterfaces [&NiagaraDataInterfaceVectorCurve, &NiagaraDataInterfaceCurve, &NiagaraDataInterfaceSocketReader]
 usedDynamicInputs []
 usedModules [
  &self:SystemSpawnScript.SystemScriptSource.SystemScriptGraph.NiagaraNodeAssignment_1.SetVariables_#1
  &SystemState
 ]
}
emitters {
 Particle_Glow {
  emitterUpdateScript [
   &EmitterState(Life Cycle Mode=::NewEnumerator1, Loop Behavior=::NewEnumerator1, Loop Duration=2, UseLoopDelay=false)
   &SpawnBurst_Instantaneous(Use Spawn Probability=false, UseLoopCountLimit=false)
  ]
  particleSpawnScript [
   &InitializeParticle(
    Color=<rgba 1 1 1 0.5>
    Lifetime Max=1
    Lifetime Min=0.9
    Lifetime Mode=::NewEnumerator1
    Sprite Size Mode=::NewEnumerator3
    Uniform Sprite Size=15
    UsePositionOffset=false
    Write Material Random=true
   )
   &SocketLocation(SamplePosOffset=false, Socket Reader=$System.Socket Reader, Socket Sampling Mode=::NewEnumerator3, VelocitySpeedLimit=false)
  ]
  particleUpdateScript [
   &ParticleState()
   &ScaleSpriteSize(Uniform Curve Sprite Scale=DataInterface() {propertyValues {Curve {keys [
        {arriveTangent=9.0417, interpMode=RCIM_Cubic, leaveTangent=9.0417, tangentMode=RCTM_User, time=0.05, value=0.6}
        {arriveTangent=2.0816, interpMode=RCIM_Cubic, leaveTangent=2.0816, tangentMode=RCTM_User, time=0.11, value=0.99}
        <at 1 1.8 RCIM_Cubic>
       ]}}})
   &ScaleColor(
    Scale Alpha=&FloatFromCurve(FloatCurve=DataInterface() {propertyValues {Curve {keys [
         {arriveTangent=43.0401, interpMode=RCIM_Cubic, leaveTangent=43.0401, tangentMode=RCTM_User, time=0, value=0}
         {arriveTangent=0.1136, interpMode=RCIM_Cubic, leaveTangent=0.1136, tangentMode=RCTM_User, time=0.02, value=1}
         {arriveTangent=-0.4334, interpMode=RCIM_Cubic, leaveTangent=-0.4334, tangentMode=RCTM_User, time=0.1, value=0.18}
         <at 1.0011 0.002 RCIM_Cubic>
        ]}}})
    Scale RGB=&VectorFromCurve(VectorCurve=DataInterface() {propertyValues {
       XCurve {keys [
         {arriveTangent=-16.1678, interpMode=RCIM_Cubic, leaveTangent=-16.1678, tangentMode=RCTM_User, time=0, value=1.15}
         {arriveTangent=-1.15, interpMode=RCIM_Cubic, leaveTangent=-1.15, time=0.1, value=0.29}
         <at 1 0 RCIM_Cubic>
        ]}
       YCurve {keys [
         {arriveTangent=-106.9967, interpMode=RCIM_Cubic, leaveTangent=-106.9967, tangentMode=RCTM_User, time=0, value=7.3}
         {arriveTangent=-5.9897, interpMode=RCIM_Cubic, leaveTangent=-5.9897, tangentMode=RCTM_User, time=0.1, value=1.83}
         <at 1 0.03 RCIM_Cubic>
        ]}
       ZCurve {keys [
         {arriveTangent=-106.9967, interpMode=RCIM_Cubic, leaveTangent=-106.9967, tangentMode=RCTM_User, time=0, value=12}
         {arriveTangent=-10.0534, interpMode=RCIM_Cubic, leaveTangent=-10.0534, tangentMode=RCTM_User, time=0.1, value=3}
         <at 1 0.09 RCIM_Cubic>
        ]}
      }})
    ScaleA=true
    ScaleRGB=true
   )
   &SolveForcesAndVelocity(Clamp Velocity=false, Limit Acceleration=false)
   &ScaleColor(
    Scale Alpha=&FNTimeOfDayScalar(Night Scale=0.5, Morning Scale=0.9, Evening Scale=0.9)
    ScaleA=true
    ScaleRGB=true
   ) {module=ScaleColor002}
   &FN_Light_Attributes(
    Color=$Particles.Color
    Create Color Attribute=true
    Create Exponent Attribute=true
    Create Position Attribute=true
    Create Radius Attribute=true
    Exponent=12
    Radius=180
    Scale Color=&FloatFromCurve(FloatCurve=DataInterface() {propertyValues {Curve {
        defaultValue=1
        keys [{interpMode=RCIM_Cubic, tangentMode=RCTM_User, time=0, value=1}, {interpMode=RCIM_Cubic, tangentMode=RCTM_User, time=1, value=0}]
       }}})
   )
  ]
  rendererClasses [&NiagaraSpriteRendererProperties, &NiagaraLightRendererProperties]
  renderers [
   {rendererClass=&NiagaraSpriteRendererProperties}
   {rendererClass=&NiagaraLightRendererProperties, rendererIndex=1}
  ]
 }
 Ribbon_SwingSmear {
  emitterSpawnScript [&FN_RibbonIDOffsetUtility()]
  emitterUpdateScript [
   &EmitterState()
   &FN_SpawnAnimTrail(AnimTrailWidthMode=::FromFirst)
  ]
  particleSpawnScript [
   &FN_InitializeAnimTrailBoneIndecies(Second Bone or Socket Index=0)
   &FN_InitializeAnimTrailBoneIndecies(AnimTrail Selection=1, First Bone or Socket Index=2, Second Bone or Socket Index=3) {enabled=false, module=FN_InitializeAnimTrailBoneIndecies001}
   &FN_InitializeAnimTrail(AnimTrailWidth=10, Color=<rgba 0.0222 0.0887 0.3372 1>, Lifetime=0.15)
   &AddVelocity(Velocity=<0 -400 0>) {enabled=false}
  ]
  particleUpdateScript [
   &SolveForcesAndVelocity(Clamp Velocity=false, Limit Acceleration=false)
   &FN_UpdateAnimTrail() {module=FN_UpdateAnimTrail001}
   &ScaleColor(
    Scale Alpha=&FloatFromCurve(FloatCurve=DataInterface() {propertyValues {Curve {keys [{arriveTangent=-1.4676, interpMode=RCIM_Cubic, leaveTangent=-1.4676, tangentMode=RCTM_User, time=0.1, value=1}, <at 1 0 RCIM_Cubic>]}}})
    ScaleA=true
    ScaleRGB=true
   )
   &ScaleColor(
    Scale Alpha=&FNTimeOfDayScalar(Night Scale=0.5, Morning Scale=0.9, Evening Scale=0.9)
    ScaleA=true
    ScaleRGB=true
   ) {module=ScaleColor002}
  ]
  rendererClasses [&NiagaraRibbonRendererProperties]
  renderers [
   {rendererClass=&NiagaraRibbonRendererProperties}
   {rendererClass=&NiagaraRibbonRendererProperties, rendererIndex=1}
  ]
 }
 Ribbon_Trail {
  emitterSpawnScript [&FN_RibbonIDOffsetUtility()]
  emitterUpdateScript [
   &EmitterState()
   &FN_SpawnAnimTrail()
  ]
  particleSpawnScript [
   &FN_InitializeAnimTrailBoneIndecies(First Bone or Socket Index=2)
   &FN_InitializeAnimTrailBoneIndecies(AnimTrail Selection=1, First Bone or Socket Index=2, Second Bone or Socket Index=3) {module=FN_InitializeAnimTrailBoneIndecies001}
   &FN_InitializeAnimTrail(Color=<rgba 3 3 3 1>, Lifetime=0.25)
   &AddVelocity(Velocity=<0 -400 0>) {enabled=false}
  ]
  particleUpdateScript [
   &SolveForcesAndVelocity(Clamp Velocity=false, Limit Acceleration=false)
   &FN_UpdateAnimTrail() {module=FN_UpdateAnimTrail001}
   &ScaleColor(
    Scale Alpha=&FloatFromCurve(FloatCurve=DataInterface() {propertyValues {Curve {keys [{arriveTangent=-2.9599, interpMode=RCIM_Cubic, leaveTangent=-2.9599, tangentMode=RCTM_User, time=0, value=1}, <at 1 0 RCIM_Cubic>]}}})
    Scale RGB=&VectorFromCurve(VectorCurve=DataInterface() {propertyValues {
       XCurve {keys [<at 0 0.14 RCIM_Cubic>,<at 1 0.02 RCIM_Cubic>]}
       YCurve {keys [<at 0 0.91 RCIM_Cubic>,<at 1 0.15 RCIM_Cubic>]}
       ZCurve {keys [<at 0 1.5 RCIM_Cubic>,<at 1 0.35 RCIM_Cubic>]}
      }})
    ScaleA=true
    ScaleRGB=true
   )
   /SetVariables(Particles.BaseTrailColor:LinearColor=&MakeLinearColorFromVectorAndFloat(
     "Vector (RGB)"=&VectorFromCurve(VectorCurve=DataInterface() {propertyValues {
        XCurve {keys [<at 0 0.0933 RCIM_Cubic>,<at 1 0.0133 RCIM_Cubic>]}
        YCurve {keys [<at 0 0.6067 RCIM_Cubic>,<at 1 0.1 RCIM_Cubic>]}
        ZCurve {keys [<at 0 1 RCIM_Cubic>,<at 1 0.2333 RCIM_Cubic>]}
       }})
     "Float (Alpha)"=&FloatFromCurve(FloatCurve=DataInterface() {propertyValues {Curve {keys [{arriveTangent=-2.9599, interpMode=RCIM_Cubic, leaveTangent=-2.9599, tangentMode=RCTM_User, time=0, value=1}, <at 1 0 RCIM_Cubic>]}}})
    )) {isSetParameters=true}
   &ScaleColor(
    Scale Alpha=&FNTimeOfDayScalar(Night Scale=0.5, Morning Scale=0.9, Evening Scale=0.9)
    ScaleA=true
    ScaleRGB=true
   ) {module=ScaleColor002}
  ]
  rendererClasses [&NiagaraRibbonRendererProperties]
  renderers [
   {rendererClass=&NiagaraRibbonRendererProperties}
   {rendererClass=&NiagaraRibbonRendererProperties, rendererIndex=1}
   {rendererClass=&NiagaraRibbonRendererProperties, rendererIndex=2}
  ]
 }
}
summary {
 systemName=NS_Pickaxe_CarbideBlue_Trail
 emitters [
  {emitterName=Ribbon_Trail, rendererClasses [&NiagaraRibbonRendererProperties]}
  {emitterName=Particle_Glow, rendererClasses [&NiagaraSpriteRendererProperties, &NiagaraLightRendererProperties]}
  {emitterName=Ribbon_SwingSmear, rendererClasses [&NiagaraRibbonRendererProperties]}
 ]
 userVariables [
  User.bIsFrontEnd {type=NiagaraBool}
  User.bIsFrontEndPreview {type=NiagaraBool}
  User.RibbonIDOffset {type=NiagaraInt32}
 ]
}
